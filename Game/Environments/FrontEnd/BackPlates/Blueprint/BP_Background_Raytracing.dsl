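; Front-end background backplate Blueprint that displays a configurable environment image on an inverted sphere mesh, using materials named for raytracing.
; The construction script swaps the half-sphere component to the inverted sphere backplate mesh when FullSphere is set.
; It then creates a dynamic material instance from the cubemap or 2D-texture material per UseCubemap, applying brightness, fade color and the matching texture.
; Its BeginPlay, overlap and Tick events are empty.
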(event EventBeginPlay)

(event Collision|EventActorBeginOverlap
  (params (OtherActor :type Actor%20Object%20Reference)))

(event EventTick
  (params (DeltaSeconds :type Float%20%28single-precision%29)))

(fn ConstructionScript
  (Utilities|FlowControl|Sequence
    (:then_0
      (if (Variables|Default|GetFullSphere)
        (Components|StaticMesh|SetStaticMesh (Variables|Default|GetSM_InvertedSphere_BackPlate_Half) "/Game/Environments/FrontEnd/Meshes/SM_InvertedSphere_BackPlate.SM_InvertedSphere_BackPlate")))
    (:then_1
      (bind _returnvalue (Rendering|Material|CreateDynamicMaterialInstance (Variables|Default|GetSM_InvertedSphere_BackPlate_Half) 0 (select (Variables|Default|GetUseCubemap) "/Game/Environments/FrontEnd/Materials/M_Background_Raytracing.M_Background_Raytracing" "/Game/Environments/FrontEnd/Materials/M_Background_Raytracing_2Dtex.M_Background_Raytracing_2Dtex") "None" :!class /Script/Engine.PrimitiveComponent))
      (Rendering|Material|SetScalarParameterValue _returnvalue "Brightness" (Variables|Default|GetBrightness) :!class /Script/Engine.MaterialInstanceDynamic)
      (Rendering|Material|SetVectorParameterValue _returnvalue "FadeColor" (Variables|Default|GetFadeColor) :!class /Script/Engine.MaterialInstanceDynamic)
      (if (Variables|Default|GetUseCubemap)
        (Rendering|Material|SetTextureParameterValue _returnvalue "Cubemap" (Variables|Default|GetCubemap))
        (else
          (Rendering|Material|SetTextureParameterValue _returnvalue "Texture" (Variables|Default|Get2dTexture)))))))
