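; Frontend lobby backplate actor that renders the background, fog and floor and controls lobby UI color schemes.
; Construction builds dynamic materials; blur materials draw into render targets once, or per tick for realtime updates.
; Temporary Galileo event logic drives a hyperspace timeline and audio from matchmaking events.
; It also writes special-event colors to a material parameter collection, toggles tagged floating shapes and sets fake floor reflections by effects quality.
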(event Custom|PlayUISpecialEventTransition
  (params (bTransitionToSpecialEvent :type Boolean))
  ;[0 0 0] UI Special Event Transition Control
  (if bTransitionToSpecialEvent
    (bind (_direction _usinguispecialeventcolors) (Timeline|UISpecailEventTransition PlayFromStart)
      (:Update
        (Rendering|Material|SetScalarParameterValue "/Game/Athena/UI/Frontend/Art/MPC_FrontendUI.MPC_FrontendUI" "UseSpecialEventColors" _usinguispecialeventcolors :!class /Script/Engine.KismetMaterialLibrary)))
    (else
      (Timeline|UISpecailEventTransition Reverse))))

(event EventMatchmakingEnded
  (params (bWasSuccess :type Boolean))
  (bind _loopingaudio (Variables|Default|GetLoopingAudio))
  ;[0 0 0] Galileo anim stuff (can remove after event)
  (if (Variables|Default|GetAllowMatchmakingInteraction)
    (if bWasSuccess
      ;;[0 0 0] success
      (Utilities|IsValid _loopingaudio
        (:Is%20Valid
          (Audio|Components|Audio|Stop _loopingaudio)
          (bind _returnvalue (Audio|CreateSound2D "/Game/Athena/Sounds/Events/Galileo/Lobby/Galileo_Lobby_Success_Cue.Galileo_Lobby_Success_Cue" 1.0 1.0 0.0 0 true))
          (Audio|Components|Audio|Play _returnvalue 0.0)))
      (else
        ;;[0 0 0] fail, reset the anim/sounds
        (CallFunction|ResetMatchmakingAnim)))))

(event Custom|SetAllowMatchmakingInteraction
  (params (AllowInteraction :type Boolean))
  (bind _loopingaudio (Variables|Default|GetLoopingAudio))
  ;[0 0 0] Galileo anim stuff (can remove after event)
  (Variables|Default|SetAllowMatchmakingInteraction AllowInteraction)
  ;;[0 0 0] a lobby transition could leave this in transit, safety check to kill looping audio
  (Utilities|IsValid _loopingaudio
    (:Is%20Valid
      (Audio|Components|Audio|Stop _loopingaudio))))

(event Custom|HandleSpecialEventAudioOnMatchmakingEnd
  (params (Success :type Boolean))
  (bind _loopingaudio (Variables|Default|GetLoopingAudio))
  ;[0 0 0] Galileo anim stuff (can remove after event)
  (if (Variables|Default|GetAllowMatchmakingInteraction)
    (if (stale EventMatchmakingEnded bWasSuccess)
      ;;[0 0 0] success
      (Utilities|IsValid _loopingaudio
        (:Is%20Valid
          (Audio|Components|Audio|Stop _loopingaudio)
          (bind _returnvalue (Audio|CreateSound2D "/Game/Athena/Sounds/Events/Galileo/Lobby/Galileo_Lobby_Success_Cue.Galileo_Lobby_Success_Cue" 1.0 1.0 0.0 0 true))
          (Audio|Components|Audio|Play _returnvalue 0.0)))
      (else
        ;;[0 0 0] fail, reset the anim/sounds
        (CallFunction|ResetMatchmakingAnim)))))

(event Custom|UpdateSettings
  (bind _returnvalue (Settings|GetGameUserSettings))
  (Rendering|Material|SetScalarParameterValue (Variables|Default|GetMID_Floor) "UseFakeReflections" (Math|Float|SelectFloat 1.0 0.0 (<= (Settings|GetVisualEffectQuality _returnvalue) 1)) :!class /Script/Engine.MaterialInstanceDynamic))

(event Custom|ResetMatchmakingAnim
  (bind _loopingaudio (Variables|Default|GetLoopingAudio))
  (bind _mid_background (Variables|Default|GetMID_Background))
  ;[0 0 0] Galileo anim stuff (can remove after event)
  (if (Variables|Default|GetAllowMatchmakingInteraction)
    (Components|Timeline|Stop (Variables|BP_FrontendBackPlate|GetHyperSpace))
    (Components|Timeline|PlayfromStart (Variables|BP_FrontendBackPlate|GetStopHyperSpeed))
    (Rendering|Material|SetScalarParameterValue (Variables|Default|GetMID_Preblur) "Blur_Color_Override" 0.0 :!class /Script/Engine.MaterialInstanceDynamic)
    (Rendering|Material|SetScalarParameterValue _mid_background "ShowStarTunnel" 0.0 :!class /Script/Engine.MaterialInstanceDynamic)
    (Rendering|Material|SetScalarParameterValue _mid_background "Flash" 0.0 :!class /Script/Engine.MaterialInstanceDynamic)
    (Rendering|Material|SetScalarParameterValue _mid_background "StopTime" 1.0 :!class /Script/Engine.MaterialInstanceDynamic)
    (Rendering|Material|SetScalarParameterValue _mid_background "HyperSpace_Transition" 1.0 :!class /Script/Engine.MaterialInstanceDynamic)
    (Rendering|Material|SetScalarParameterValue _mid_background "StarFieldBrightness" 1.0 :!class /Script/Engine.MaterialInstanceDynamic)
    (Rendering|Material|SetScalarParameterValue _mid_background "FadeStarField" 0.0 :!class /Script/Engine.MaterialInstanceDynamic)
    ;;[0 0 0] matchmake fail (cancel or unable to find match)
    (Utilities|IsValid _loopingaudio
      (:Is%20Valid
        (Audio|Components|Audio|Stop _loopingaudio)
        (Audio|PlaySound2D "/Game/Athena/Sounds/Events/Galileo/Lobby/Galileo_Lobby_Fail_Cue.Galileo_Lobby_Fail_Cue")))))

(event EventEnteringMatchmaking
  (bind _mid_background (Variables|Default|GetMID_Background))
  (bind _loopingaudio (Variables|Default|GetLoopingAudio))
  (bind _mid_background_2 (Variables|Default|GetMID_Background))
  ;[0 0 0] Galileo anim stuff (can remove after event)
  ;[0 0 0] enter hyperspeed on matchmaking enter
  (if (Variables|Default|GetAllowMatchmakingInteraction)
    (Utilities|FlowControl|Sequence
      (:then_0
        (Rendering|Material|SetScalarParameterValue (Variables|Default|GetMID_Preblur) "Blur_Color_Override" 0.0 :!class /Script/Engine.MaterialInstanceDynamic)
        (Rendering|Material|SetScalarParameterValue _mid_background_2 "ShowStarTunnel" 0.0 :!class /Script/Engine.MaterialInstanceDynamic)
        (Rendering|Material|SetScalarParameterValue _mid_background_2 "Flash" 0.0 :!class /Script/Engine.MaterialInstanceDynamic)
        (Rendering|Material|SetScalarParameterValue _mid_background_2 "StopTime" 1.0 :!class /Script/Engine.MaterialInstanceDynamic)
        (Rendering|Material|SetScalarParameterValue _mid_background_2 "HyperSpace_Transition" 1.0 :!class /Script/Engine.MaterialInstanceDynamic)
        (Rendering|Material|SetScalarParameterValue _mid_background_2 "StarFieldBrightness" 1.0 :!class /Script/Engine.MaterialInstanceDynamic)
        (Rendering|Material|SetScalarParameterValue _mid_background_2 "FadeStarField" 0.0 :!class /Script/Engine.MaterialInstanceDynamic)
        ;;[0 0 0] matchmake fail (cancel or unable to find match)
        (Utilities|IsValid _loopingaudio
          (:Is%20Valid
            (Audio|Components|Audio|Stop _loopingaudio)
            (Audio|PlaySound2D "/Game/Athena/Sounds/Events/Galileo/Lobby/Galileo_Lobby_Fail_Cue.Galileo_Lobby_Fail_Cue"))))
      (:then_1
        (Utilities|FlowControl|Delay 1.0)
        (Rendering|Material|SetScalarParameterValue (Variables|Default|GetMID_Background) "IsStartingHyperspace" 1.0 :!class /Script/Engine.MaterialInstanceDynamic)
        (bind _returnvalue (Audio|SpawnSound2D "/Game/Athena/Sounds/Events/Galileo/Lobby/Galileo_Lobby_Loop_Cue.Galileo_Lobby_Loop_Cue"))
        (Variables|Default|SetLoopingAudio _returnvalue)
        (Rendering|Material|SetScalarParameterValue (Variables|Default|GetMID_Preblur) "BlurControl" 1.0 :!class /Script/Engine.MaterialInstanceDynamic)
        (bind (_direction _starlines _flash _tunnel _bloom _showstarlines _uvscale _numstarloops _starbloom _starfieldfade _staruvscale) (Timeline|HyperSpace PlayFromStart)
          (:Update
            (Rendering|Material|SetScalarParameterValue (Variables|Default|GetMID_Preblur) "Blur_Color_Override" _staruvscale :!class /Script/Engine.MaterialInstanceDynamic)
            (Rendering|Material|SetScalarParameterValue _mid_background "ShowStarTunnel" _showstarlines :!class /Script/Engine.MaterialInstanceDynamic)
            (Rendering|Material|SetScalarParameterValue _mid_background "HyperSpace_Transition" (- 1.0 _starlines) :!class /Script/Engine.MaterialInstanceDynamic)
            (Rendering|Material|SetScalarParameterValue _mid_background "ShowStarLines" _tunnel :!class /Script/Engine.MaterialInstanceDynamic)
            (Rendering|Material|SetScalarParameterValue _mid_background "StarFieldBrightness" _starbloom :!class /Script/Engine.MaterialInstanceDynamic)
            (Rendering|Material|SetScalarParameterValue _mid_background "FadeStarField" _starfieldfade :!class /Script/Engine.MaterialInstanceDynamic)
            (Rendering|Material|SetScalarParameterValue _mid_background "UVScale" _uvscale :!class /Script/Engine.MaterialInstanceDynamic)
            (Rendering|Material|SetScalarParameterValue _mid_background "NumStarLoops" _numstarloops :!class /Script/Engine.MaterialInstanceDynamic)
            (Rendering|Material|SetScalarParameterValue _mid_background "Bloom" _bloom :!class /Script/Engine.MaterialInstanceDynamic)
            (Rendering|Material|SetScalarParameterValue _mid_background "StopTime" _tunnel :!class /Script/Engine.MaterialInstanceDynamic)
            (Rendering|Material|SetScalarParameterValue (Variables|Default|GetMID_Preblur) "BlurControl" _tunnel :!class /Script/Engine.MaterialInstanceDynamic)
            (Rendering|Material|SetScalarParameterValue _mid_background "IsStartingHyperspace" _tunnel :!class /Script/Engine.MaterialInstanceDynamic)
            (Rendering|Material|SetScalarParameterValue _mid_background "Flash" _flash :!class /Script/Engine.MaterialInstanceDynamic)))))))

(event EventBeginPlay
  (Utilities|FlowControl|Sequence
    (:then_0
      (bind _returnvalue (Settings|GetGameUserSettings))
      (EventDispatchers|BindEventtoOnGameUserSettingsUINeedsUpdate _returnvalue (delegate Custom|UpdateSettings))
      (Rendering|Material|SetScalarParameterValue (Variables|Default|GetMID_Floor) "UseFakeReflections" (Math|Float|SelectFloat 1.0 0.0 (<= (Settings|GetVisualEffectQuality _returnvalue) 1)) :!class /Script/Engine.MaterialInstanceDynamic))
    (:then_1
      (CallFunction|UpdateRenderTarget))))

(event Custom|PlayEnterMatchmakingAnim
  (bind _mid_background (Variables|Default|GetMID_Background))
  (bind _loopingaudio (Variables|Default|GetLoopingAudio))
  (bind _mid_background_2 (Variables|Default|GetMID_Background))
  ;[0 0 0] Galileo anim stuff (can remove after event)
  ;[0 0 0] enter hyperspeed on matchmaking enter
  (if (Variables|Default|GetAllowMatchmakingInteraction)
    (Utilities|FlowControl|Sequence
      (:then_0
        (Rendering|Material|SetScalarParameterValue (Variables|Default|GetMID_Preblur) "Blur_Color_Override" 0.0 :!class /Script/Engine.MaterialInstanceDynamic)
        (Rendering|Material|SetScalarParameterValue _mid_background_2 "ShowStarTunnel" 0.0 :!class /Script/Engine.MaterialInstanceDynamic)
        (Rendering|Material|SetScalarParameterValue _mid_background_2 "Flash" 0.0 :!class /Script/Engine.MaterialInstanceDynamic)
        (Rendering|Material|SetScalarParameterValue _mid_background_2 "StopTime" 1.0 :!class /Script/Engine.MaterialInstanceDynamic)
        (Rendering|Material|SetScalarParameterValue _mid_background_2 "HyperSpace_Transition" 1.0 :!class /Script/Engine.MaterialInstanceDynamic)
        (Rendering|Material|SetScalarParameterValue _mid_background_2 "StarFieldBrightness" 1.0 :!class /Script/Engine.MaterialInstanceDynamic)
        (Rendering|Material|SetScalarParameterValue _mid_background_2 "FadeStarField" 0.0 :!class /Script/Engine.MaterialInstanceDynamic)
        ;;[0 0 0] matchmake fail (cancel or unable to find match)
        (Utilities|IsValid _loopingaudio
          (:Is%20Valid
            (Audio|Components|Audio|Stop _loopingaudio)
            (Audio|PlaySound2D "/Game/Athena/Sounds/Events/Galileo/Lobby/Galileo_Lobby_Fail_Cue.Galileo_Lobby_Fail_Cue"))))
      (:then_1
        (Utilities|FlowControl|Delay 1.0)
        (Rendering|Material|SetScalarParameterValue (Variables|Default|GetMID_Background) "IsStartingHyperspace" 1.0 :!class /Script/Engine.MaterialInstanceDynamic)
        (bind _returnvalue (Audio|SpawnSound2D "/Game/Athena/Sounds/Events/Galileo/Lobby/Galileo_Lobby_Loop_Cue.Galileo_Lobby_Loop_Cue"))
        (Variables|Default|SetLoopingAudio _returnvalue)
        (Rendering|Material|SetScalarParameterValue (Variables|Default|GetMID_Preblur) "BlurControl" 1.0 :!class /Script/Engine.MaterialInstanceDynamic)
        (Timeline|HyperSpace PlayFromStart)))))

(event Custom|DebugScalabilitySettings
  (Utilities|FlowControl|Sequence
    (:then_0
      (bind _returnvalue (Settings|GetGameUserSettings))
      (EventDispatchers|BindEventtoOnGameUserSettingsUINeedsUpdate _returnvalue (delegate Custom|UpdateSettings))
      (Rendering|Material|SetScalarParameterValue (Variables|Default|GetMID_Floor) "UseFakeReflections" (Math|Float|SelectFloat 1.0 0.0 (<= (Settings|GetVisualEffectQuality _returnvalue) 1)) :!class /Script/Engine.MaterialInstanceDynamic))
    (:then_1
      (CallFunction|UpdateRenderTarget))))

(event Custom|TestBlur
  (CallFunction|UpdateRenderTarget))

(event EventTick
  (params (DeltaSeconds :type Float%20%28single-precision%29))
  (CallFunction|UpdateRenderTargetRealtime))

(fn CreateUniqueMaterialName
  (params (PostFix :type String))
  (return (Utilities|String|Append (Utilities|String|Append (Utilities|GetDisplayName self) (Utilities|String|ToString(Integer) (Math|Float|Truncate (/ (Math|Float|float+float (.x (Transformation|GetActorLocation self)) (.y (Transformation|GetActorLocation self)) (.z (Transformation|GetActorLocation self))) 3.0)))) PostFix)))

(fn EnableRealtimeUpdate
  (params (Enabled :type Boolean))
  (Actor|Tick|SetActorTickEnabled Enabled)
  (Rendering|Material|SetTextureParameterValue (Variables|Default|GetBlurMaterial) "Param" (select Enabled (Variables|Default|GetRenderTarget_preblur) (Variables|Default|GetBackgroundTexture))))

(fn GetUIColorScheme
  (return (Variables|UIDisplay|GetCurrentUIColorScheme) (Variables|UIDisplay|GetbHasUIColorOverride)))

(fn HasMatcmakingInteraction
  (return (Variables|Default|GetAllowMatchmakingInteraction)))

(fn HasUIColorOverride
  (return (Variables|UIDisplay|GetbHasUIColorOverride)))

(fn RequestUIColorScheme
  (params (MainUIColor :type Linear%20Color%20Structure))
  (Variables|UIDisplay|SetbHasUIColorOverride true)
  (bind _output_get (Variables|UIDisplay|SetCurrentUIColorScheme MainUIColor))
  (Default|CallRequestLobbyColorChanges _output_get (not (Variables|UIDisplay|GetbHasUIColorOverride))))

(fn ResetUIColorScheme
  (bind _output_get (Variables|UIDisplay|SetbHasUIColorOverride false))
  (Default|CallRequestLobbyColorChanges :UseDefaultScheme (not _output_get)))

(fn SetBackgroundTexture
  (params (InTexture :type Texture%20Object%20Reference) (InMaterial :type Material%20Interface%20Object%20Reference) (InPreBlurMaterial :type Material%20Interface%20Object%20Reference))
  (Utilities|FlowControl|Sequence
    (:then_0
      (Variables|Default|SetMaterial_background (select (Utilities|IsValid (Variables|GetInMaterial) :!class /Script/Engine.KismetSystemLibrary) (Variables|GetInMaterial) (Variables|Default|GetMaterial_background)))
      (Variables|Default|SetPreBlurMaterial (select (Utilities|IsValid (Variables|GetInPreBlurMaterial) :!class /Script/Engine.KismetSystemLibrary) (Variables|GetInPreBlurMaterial) "/Game/Environments/FrontEnd/BackPlates/Materials/M_FrontendBackplate_Master1_Blur.M_FrontendBackplate_Master1_Blur"))
      (bind _returnvalue (Rendering|Material|CreateDynamicMaterialInstance (Variables|Default|GetStaticMesh) 0 (Variables|Default|GetMaterial_background) (Utilities|String|StringToName (CallFunction|CreateUniqueMaterialName "_D")) :!class /Script/Engine.PrimitiveComponent))
      (bind _output_get (Variables|Default|SetMID_Background _returnvalue))
      (Rendering|Material|SetMaterial (Variables|Default|GetSM_INvertedSphere) 0 _output_get)
      (bind _output_get_2 (LocalVariables|SetLocalMIDTexture (select (Utilities|IsValid (Variables|GetInTexture) :!class /Script/Engine.KismetSystemLibrary) (Variables|GetInTexture) (Variables|Default|GetBackgroundTexture))))
      (Rendering|Material|SetTextureParameterValue (Variables|Default|GetMID_Background) "param" _output_get_2)
      (bind _returnvalue_2 (Rendering|Material|CreateDynamicMaterialInstance (Variables|Default|GetStaticMesh_fog) 0 0 (Utilities|String|StringToName (CallFunction|CreateUniqueMaterialName "_E")) :!class /Script/Engine.PrimitiveComponent))
      (Rendering|Material|SetTextureParameterValue _returnvalue_2 "param" (LocalVariables|GetLocalMIDTexture))
      (bind _returnvalue_3 (Rendering|Material|CreateDynamicMaterialInstance (Variables|Default|GetPreBlurMaterial) (Utilities|String|StringToName (CallFunction|CreateUniqueMaterialName "_F")) "None" :!class /Script/Engine.KismetMaterialLibrary))
      (bind _output_get_3 (Variables|Default|SetMID_Preblur _returnvalue_3))
      (Rendering|Material|SetTextureParameterValue _output_get_3 "param" (LocalVariables|GetLocalMIDTexture))
      (Rendering|Material|SetTextureParameterValue (Variables|Default|GetBlurMaterial) "param" (LocalVariables|GetLocalMIDTexture)))
    (:then_1
      (Utilities|IsValid (Variables|GetInTexture)
        (:Is%20Valid
          (Variables|Default|SetBackgroundTexture (Variables|GetInTexture)))))))

(fn SetFloatingShapeVisibility
  (params (Show :type Boolean))
  (for _array_element (Variables|Default|GetFloatingShapes)
    (Utilities|IsValid _array_element
      (:Is%20Valid
        (Rendering|SetActorHiddenInGame _array_element (not Show))))))

(fn SetSpecialEventUIColorScheme
  (params (SetUIPanelColors :type Boolean) (Backing :type Linear%20Color%20Structure) (Backing_Hover :type Linear%20Color%20Structure) (Backing_Selection :type Linear%20Color%20Structure) (SetFontColors :type Boolean) (Primary :type Linear%20Color%20Structure) (Secondary :type Linear%20Color%20Structure) (Tertiary :type Linear%20Color%20Structure) (Quaternary :type Linear%20Color%20Structure))
  (Utilities|FlowControl|Sequence
    (:then_0
      (if SetUIPanelColors
        (Utilities|FlowControl|Sequence
          (:then_0
            (Rendering|Material|SetVectorParameterValue "/Game/Athena/UI/Frontend/Art/MPC_FrontendUI.MPC_FrontendUI" "Special_RGBA_BackingColor" Backing :!class /Script/Engine.KismetMaterialLibrary))
          (:then_1
            (Rendering|Material|SetVectorParameterValue "/Game/Athena/UI/Frontend/Art/MPC_FrontendUI.MPC_FrontendUI" "Special_RGBA_BackingColor_Hover" Backing_Hover :!class /Script/Engine.KismetMaterialLibrary))
          (:then_2
            (Rendering|Material|SetVectorParameterValue "/Game/Athena/UI/Frontend/Art/MPC_FrontendUI.MPC_FrontendUI" "Special_RGBA_BackingColor_Selected" Backing_Selection :!class /Script/Engine.KismetMaterialLibrary)))))
    (:then_1
      (if SetFontColors
        (Utilities|FlowControl|Sequence
          (:then_0
            (Rendering|Material|SetVectorParameterValue "/Game/Athena/UI/Frontend/Art/MPC_FrontendUI.MPC_FrontendUI" "Special_FontColor_Primary" Primary :!class /Script/Engine.KismetMaterialLibrary))
          (:then_1
            (Rendering|Material|SetVectorParameterValue "/Game/Athena/UI/Frontend/Art/MPC_FrontendUI.MPC_FrontendUI" "Special_FontColor_Secondary" Secondary :!class /Script/Engine.KismetMaterialLibrary))
          (:then_2
            (Rendering|Material|SetVectorParameterValue "/Game/Athena/UI/Frontend/Art/MPC_FrontendUI.MPC_FrontendUI" "Special_FontColor_Tertiary" Tertiary :!class /Script/Engine.KismetMaterialLibrary))
          (:then_3
            (Rendering|Material|SetVectorParameterValue "/Game/Athena/UI/Frontend/Art/MPC_FrontendUI.MPC_FrontendUI" "Special_FontColor_Quaternary" Quaternary :!class /Script/Engine.KismetMaterialLibrary)))))))

(fn UpdateRenderTarget
  (Rendering|DrawMaterialtoRenderTarget (Variables|Default|GetRenderTarget) (Variables|Default|GetBlurMaterial)))

(fn UpdateRenderTargetRealtime
  (Rendering|DrawMaterialtoRenderTarget (Variables|Default|GetRenderTarget_preblur) (Variables|Default|GetMID_Preblur))
  (CallFunction|UpdateRenderTarget))

(fn ConstructionScript
  (Utilities|FlowControl|Sequence
    (:then_0
      ;[1 1 1] Set Materials
      (Rendering|Material|SetMaterial (Variables|Default|GetStaticMesh) 0 (Variables|Default|GetMaterial_background))
      (Rendering|Material|SetMaterial (Variables|Default|GetStaticMesh_fog) 0 (Variables|Default|GetMaterial_Fog))
      (bind _returnvalue (Rendering|Material|CreateDynamicMaterialInstance "/Game/Environments/FrontEnd/BackPlates/Materials/M_FrontendBackplate_Master_Blur.M_FrontendBackplate_Master_Blur" (Utilities|String|StringToName (CallFunction|CreateUniqueMaterialName "_A")) "None" :!class /Script/Engine.KismetMaterialLibrary))
      (Variables|Default|SetBlurMaterial _returnvalue)
      (bind _returnvalue_2 (Rendering|Material|CreateDynamicMaterialInstance (Variables|Default|GetPreBlurMaterial) (Utilities|String|StringToName (CallFunction|CreateUniqueMaterialName "_B")) "None" :!class /Script/Engine.KismetMaterialLibrary))
      (Variables|Default|SetMID_Preblur _returnvalue_2))
    (:then_1
      (CallFunction|SetBackgroundTexture ""))
    (:then_2
      ;[1 1 1] Realtime update
      (Actor|Tick|SetActorTickEnabled (Variables|Default|GetUpdateFoginRealTime)))
    (:then_3
      (bind _returnvalue_3 (Rendering|Material|CreateDynamicMaterialInstance (Variables|Default|GetSM_FrontendBackPlate_Floor) 0 0 (Utilities|String|StringToName (CallFunction|CreateUniqueMaterialName "_C")) :!class /Script/Engine.PrimitiveComponent))
      (Variables|Default|SetMID_Floor _returnvalue_3))
    (:then_4
      (bind _outactors (Actor|GetAllActorswithTag "AmbientFloating"))
      (bind _outsortedactors (CookingNon-Determinism|SortActorArray _outactors))
      (Variables|Default|SetFloatingShapes _outsortedactors))))
