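; Blueprint interface for front-end backplates that react to matchmaking.
; It declares events for entering matchmaking and for matchmaking ending, the latter passing a success flag.
; A query, spelled HasMatcmakingInteraction, returns a boolean for whether a backplate has such an interaction, false by default.
; The file holds only these signatures, with no reaction logic of its own.
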(fn EnteringMatchmaking)

(fn HasMatcmakingInteraction
  (return false))

(fn MatchmakingEnded
  (params (bWasSuccess :type Boolean)))
