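# Translucent unlit front-end backplate material that draws a donut store icon on a rippling, cloth-like background.
# A custom HLSL sine wave, animated by time, both warps the UVs and blends shadow and lit colors.
# The warped donut mask's RGB channels select tinted icon colors, composited over a radial gradient background.
# A tiled cloth texture, defaulting to corrugated roof, multiplies the composite to add fabric shading.
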Material /Game/Environments/FrontEnd/BackPlates/Materials/M_DonutStore
graph {
 properties {blendMode=BLEND_Translucent, pixelDepthOffsetMode=PDOM_Legacy, shadingModel=MSM_Unlit, usedShadingModels=Unlit}
 nodes [
  n1 = VectorParameter() {parameterName=IconUVOffsetAndScale, defaultValue=<rgba 0 0 1 1>}
  MaterialExpressionAdd_5 = Add(
   A=MaterialFunctionCall(UVs=TextureCoordinate(), Texture Scale=AppendVector(A=@n1.B, B=@n1.B)) {materialFunction=&ScaleUVsByCenter}.UVs
   B=AppendVector(A=@n1.R, B=@n1.R)
  )
  MaterialExpressionTime_1 = Time()
  MaterialExpressionScalarParameter_2 = ScalarParameter() {defaultValue=0.036, parameterName=OffsetRange}
  MaterialExpressionScalarParameter_3 = ScalarParameter() {defaultValue=0.9, parameterName=TimeMult}
  MaterialExpressionScalarParameter_4 = ScalarParameter() {defaultValue=1.5, parameterName=ScaleA}
  MaterialExpressionScalarParameter_5 = ScalarParameter() {defaultValue=6.5, parameterName=AddA}
  MaterialExpressionScalarParameter_6 = ScalarParameter() {defaultValue=0.3, parameterName=TurbA}
  MaterialExpressionScalarParameter_7 = ScalarParameter() {defaultValue=0.6, parameterName=TurbB}
  n2 = Custom!(
   inUV=@MaterialExpressionAdd_5
   inTime=@MaterialExpressionTime_1
   inValRange=@MaterialExpressionScalarParameter_2
   inTimeScale=@MaterialExpressionScalarParameter_3
   sinScaleA=@MaterialExpressionScalarParameter_4
   sinAddA=@MaterialExpressionScalarParameter_5
   sinTurbA=@MaterialExpressionScalarParameter_6
   sinTurbB=@MaterialExpressionScalarParameter_7
  ) {
   code="   float wave = sin((inUV.y + inUV.x - inTime * inTimeScale + sin(sinScaleA * inUV.y + sinAddA * inUV.x) * PI * sinTurbA)\r\n                  * PI * sinTurbB);\r\n    \r\n   return wave;"
   description=Wave
   outputType=CMOT_Float1
   inputs [
    {inputName=inUV, input {expression=&self:MaterialExpressionAdd_5}}
    {inputName=inTime, input {expression=&self:MaterialExpressionTime_1}}
    {inputName=inValRange, input {expression=&self:MaterialExpressionScalarParameter_2}}
    {inputName=inTimeScale, input {expression=&self:MaterialExpressionScalarParameter_3}}
    {inputName=sinScaleA, input {expression=&self:MaterialExpressionScalarParameter_4}}
    {inputName=sinAddA, input {expression=&self:MaterialExpressionScalarParameter_5}}
    {inputName=sinTurbA, input {expression=&self:MaterialExpressionScalarParameter_6}}
    {inputName=sinTurbB, input {expression=&self:MaterialExpressionScalarParameter_7}}
   ]
  }
  n3 = Clamp(None=Multiply(A=@n2))
  n4 = LinearInterpolate(
   A=VectorParameter() {parameterName=IconCol_G_Shadow, defaultValue=<rgba 0.2188 0 0 1>}.RGB
   B=VectorParameter() {parameterName=IconCol_G, defaultValue=<rgba 1 0 0 1>}.RGB
   Alpha=@n3
  )
  n5 = Multiply(A=@MaterialExpressionScalarParameter_2, B=@n2)
  n6 = Multiply(A=Add(A=Constant() {r=1}, B=Subtract(A=@MaterialExpressionScalarParameter_2, B=@n5)), B=@MaterialExpressionAdd_5)
  n7 = TextureSample(UVs=@n6) {
   samplerSource=SSM_Clamp_WorldGroupSettings
   samplerType=SAMPLERTYPE_Masks
   texture=/Game/Environments/FrontEnd/BackPlates/Textures/T_UI_DonutIconMask
  }
  n8 = LinearInterpolate(
   A=VectorParameter() {parameterName=iconCol_R_Shadow, defaultValue=<rgba 0 0 0 1>}.RGB
   B=VectorParameter() {parameterName=iconCol_R, defaultValue=<rgba 0.0469 0.0469 0.0469 1>}.RGB
   Alpha=@n3
  )
  n9 = Multiply(A=Multiply(A=OneMinus(None=@n7.G), B=@n7.R), B=OneMinus(None=@n7.B))
  n10 = VectorParameter() {parameterName=IconCol_B_Shadow, defaultValue=<rgba 0.2865 0.2865 0.2865 1>}
  n11 = VectorParameter() {parameterName=IconCol_B, defaultValue=<rgba 1 1 1 1>}
  n12 = Add(A=Add(A=Multiply(A=@n4, B=@n7.G), B=Multiply(A=@n8, B=@n9)), B=Multiply(A=LinearInterpolate(A=@n10.RGB, B=@n11.RGB, Alpha=@n3), B=@n7.B))
  Add(B=@n12)
  Multiply(A=VectorParameter() {parameterName=ShadowIconColor, defaultValue=<rgba 0 0 0 1>}.RGB, B=@n7.R)
  n13 = TextureSampleParameter2D(UVs=Multiply(A=@n6, B=ScalarParameter() {defaultValue=2, parameterName=curtainNoiseTile}) {constB=2}) {parameterName=ClothTexture, texture=/Game/Environments/Sets/AB_Research/Textures/T_Corrugated_Roof_D}
  n14 = ScalarParameter() {defaultValue=1, parameterName=texScal}
  n15 = VectorParameter() {parameterName=GradUVOffsetAndScale, defaultValue=<rgba 0 0 1 1>}
  n16 = MaterialFunctionCall(UVs=@n6, Texture Scale=AppendVector(A=@n15.B, B=@n15.B)) {materialFunction=&ScaleUVsByCenter}
  n17 = AppendVector(A=@n15.R, B=@n15.R)
  n18 = LinearInterpolate(
   A=LinearInterpolate(
    A=VectorParameter() {parameterName=ShadowColA, defaultValue=<rgba 0.151 0 0 1>}.RGB
    B=VectorParameter() {parameterName=ShadowColB, defaultValue=<rgba 0.5365 0 0 1>}.RGB
    Alpha=@n3
   )
   B=LinearInterpolate(
    A=VectorParameter() {parameterName=GradColB, defaultValue=<rgba 0 0.3651 0.526 1>}.RGB
    B=VectorParameter() {parameterName=GradCol_A, defaultValue=<rgba 0.1302 0.038 0.038 1>}.RGB
    Alpha=@n3
   )
   Alpha=MaterialFunctionCall(
    UVs=Add(A=@n16.UVs, B=@n17)
    CenterPosition=Constant2Vector() {g=0.5, r=0.5}
    Radius=ScalarParameter() {defaultValue=0.5, parameterName=GradRad}
    Density=ScalarParameter() {defaultValue=2.33, parameterName=Density}
   ) {materialFunction=&RadialGradientExponential}.RadialGradientExponential
  )
  n19 = OneMinus(None=@n7.R)
  Multiply(A=Clamp(None=Multiply(A=@n13.R, B=@n14)), B=Add(A=Multiply(A=@n18, B=@n19), B=@n12))
  ScalarParameter() {defaultValue=1, parameterName=Opacity}
  TextureCoordinate()
 ]
}
