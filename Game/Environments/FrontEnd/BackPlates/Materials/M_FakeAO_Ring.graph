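# Unlit translucent material that fakes a ring of ambient occlusion for front-end backplates, using a dark blue color parameter.
# Two falloff gradients, each built from the U texture coordinate raised to a power and scaled by opacity, blend by the FloorStylised value from the MPC_LobbyBackground parameter collection.
# A normalized world-to-camera vector, scaled by the Zfightfix parameter, provides an offset toward the camera against z-fighting.
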Material /Game/Environments/FrontEnd/BackPlates/Materials/M_FakeAO_Ring
graph {
 properties {blendMode=BLEND_Translucent, pixelDepthOffsetMode=PDOM_Legacy, shadingModel=MSM_Unlit, usedShadingModels=Unlit}
 nodes [
  n1 = ComponentMask(None=TextureCoordinate()) {r=true}
  n2 = ScalarParameter() {defaultValue=2.3634, parameterName=Falloff}
  n3 = ComponentMask(None=TextureCoordinate()) {r=true}
  n4 = ScalarParameter() {defaultValue=5.9894, parameterName=Falloff2}
  LinearInterpolate(
   A=Multiply(A=Power(Base=@n1, Exp=@n2), B=ScalarParameter() {defaultValue=0.8857, parameterName=Opacity})
   B=Multiply(A=Power(Base=@n3, Exp=@n4), B=ScalarParameter() {defaultValue=0.9731, parameterName=Opacity2})
   Alpha=ComponentMask(None=CollectionParameter() {collection=&/MPC_LobbyBackground, parameterName=FloorStylised}) {r=true}
  )
  n5 = CameraPositionWS()
  n6 = WorldPosition()
  Multiply(A=Normalize(VectorInput=Subtract(A=@n5, B=@n6.XYZ)), B=ScalarParameter() {defaultValue=2, parameterName=Zfightfix})
  VectorParameter() {parameterName=Color, defaultValue=<rgba 0.0168 0.0271 0.0521 1>}
 ]
}
