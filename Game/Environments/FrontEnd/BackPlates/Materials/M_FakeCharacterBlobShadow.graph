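# Translucent unlit material for a fake character blob shadow in the front-end backplates, using a shadow texture scaled by Opacity.
# Vertex-offset math uses UVs to push a card sideways and along the horizontal camera-to-object axis, with clamped stretching that varies with view elevation.
# ShowShadows blends between collapsing vertices onto the object pivot and applying that offset.
# A shading-path switch allows the shadow only on Forward and Mobile, collapsing it on Default and Deferred.
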Material /Game/Environments/FrontEnd/BackPlates/Materials/M_FakeCharacterBlobShadow
graph {
 properties {blendMode=BLEND_Translucent, pixelDepthOffsetMode=PDOM_Legacy, shadingModel=MSM_Unlit, usedShadingModels=Unlit}
 nodes [
  Constant3Vector()
  Multiply(
   A=TextureSample() {texture=/Game/Environments/FrontEnd/BackPlates/Textures/T_FakeShadow}.R
   B=ScalarParameter() {defaultValue=0.9232, parameterName=Opacity}
  )
  n1 = Subtract(A=ObjectPositionWS(), B=WorldPosition() {worldPositionShaderOffset=WPT_ExcludeAllShaderOffsets}.XYZ)
  n2 = Subtract(A=ObjectPositionWS(), B=CameraPositionWS())
  n3 = Multiply(A=@n2, B=Constant3Vector() {constant=<rgba 1 1 0 0>})
  n4 = Constant3Vector() {constant=<rgba 0 0 1 0>}
  n5 = TextureCoordinate()
  n6 = ComponentMask(None=@n5) {r=true}
  n7 = Multiply(A=Normalize(VectorInput=CrossProduct(A=@n3, B=@n4)), B=Sign(None=Subtract(A=@n6) {constB=0.5}))
  n8 = ScalarParameter() {defaultValue=48.9338, parameterName=Width}
  n9 = ComponentMask(None=@n5) {g=true}
  n10 = Normalize(VectorInput=@n3)
  n11 = ScalarParameter() {defaultValue=71.0482, parameterName=Length}
  n12 = DotProduct(A=Constant3Vector() {constant=<rgba 0 0 1 0>}, B=Normalize(VectorInput=@n2))
  n13 = Multiply(
   A=Multiply(A=OneMinus(None=@n9), B=@n10)
   B=Clamp(None=Divide(A=@n11, B=@n12), Min=ScalarParameter() {defaultValue=-347.1356, parameterName=Max}) {clampMode=CMODE_ClampMin}
  )
  n14 = Multiply(A=Multiply(A=@n10, B=@n9), B=ScalarParameter() {defaultValue=28.7597, parameterName=Length2})
  n15 = LinearInterpolate(A=@n1, B=Add(A=Multiply(A=@n7, B=@n8), B=Add(A=@n13, B=@n14)), Alpha=ScalarParameter() {defaultValue=1, parameterName=ShowShadows})
  ShadingPathSwitch(Default=@n1, Deferred=@n1, Forward=@n15, Mobile=@n15)
 ]
}
