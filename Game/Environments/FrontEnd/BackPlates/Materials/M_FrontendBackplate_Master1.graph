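# Unlit master material for front-end lobby backplates, with static switches for fog and realtime-blur variants.
# Color blends a blurred render target toward screen edges, crossfades to the default lobby texture via MPC_LobbyBackground Fade, and applies a vignette.
# The fog variant uses gamma-adjusted blur, with FloorStylised shifting color, opacity and offset toward a flat floor look.
# Depth-faded masks drive opacity, and world position offsets along the camera vector, except under realtime blur.
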Material /Game/Environments/FrontEnd/BackPlates/Materials/M_FrontendBackplate_Master1
graph {
 properties {pixelDepthOffsetMode=PDOM_Legacy, shadingModel=MSM_Unlit, translucencyPass=MTP_BeforeDOF, usedShadingModels=Unlit}
 nodes [
  Constant()
  n1 = Add(A=TextureCoordinate(), B=AppendVector(A=ScalarParameter() {parameterName=UVOffset_X}, B=ScalarParameter() {parameterName=UVOffset_Y}))
  n2 = ScalarParameter() {defaultValue=1.142, parameterName=ScaleX}
  n3 = ScalarParameter() {defaultValue=0.604, parameterName=ScaleY}
  n4 = StaticSwitchParameter(
   True=@n1
   False=MaterialFunctionCall(UVs=@n1, Texture Scale=AppendVector(A=@n2, B=@n3)) {materialFunction=&ScaleUVsByCenter}.UVs
  ) {parameterName=IsRealtimeBlur}
  n5 = Constant2Vector() {g=0.5, r=0.5}
  n6 = Multiply(
   A=ScalarParameter() {defaultValue=0.8667, parameterName=VignetteIntensity}
   B=Divide(A=Distance(A=@n4, B=@n5), B=ScalarParameter() {defaultValue=1.1363, parameterName=VignetteSize})
  )
  n7 = CollectionParameter() {collection=&/MPC_LobbyBackground, parameterName=FloorStylised}
  n8 = TextureSample(UVs=@n4) {texture=/Game/Environments/FrontEnd/BackPlates/Textures/T_Background_LobbyDefault}
  n9 = CollectionParameter() {collection=&MPC_LobbyBackground, parameterName=Fade}
  n10 = LinearInterpolate(
   A=TextureSampleParameter2D(UVs=@n4) {parameterName=RenderTarget, texture=&/RT_Frontend_Backplate_Blur}.RGB
   B=@n8.RGB
   Alpha=@n9
  )
  n11 = Power(Base=@n10, Exp=ScalarParameter() {defaultValue=1, parameterName=FogGamma})
  n12 = Multiply(A=ComponentMask(None=CollectionParameter() {collection=&MPC_LobbyBackground, parameterName=FloorColor}) {b=true, g=true, r=true}) {constB=0.8}
  n13 = ComponentMask(None=CollectionParameter() {collection=&MPC_LobbyBackground, parameterName=FloorStylised}) {r=true}
  n14 = TextureSampleParameter2D!(UVs=@n4, Bias=ScalarParameter() {parameterName=ViewMipBias}, Apply View MipBias) {mipValueMode=TMVM_MipBias, parameterName=Param, texture=&T_Background_LobbyDefault}
  n15 = LinearInterpolate(A=@n14.RGB, B=@n8.RGB, Alpha=@n9)
  n16 = ComponentMask(None=@n4) {r=true}
  n17 = OneMinus(None=@n16)
  n18 = ScalarParameter() {defaultValue=7.0987, parameterName=VignetteBrightness}
  n19 = ScalarParameter() {defaultValue=1.1831, parameterName=FalloffVignette}
  n20 = Power(Base=Multiply(A=Multiply(A=@n17, B=@n16), B=@n18) {constB=2}, Exp=@n19)
  n21 = ComponentMask(None=@n4) {g=true}
  n22 = OneMinus(None=@n21)
  n23 = Power(Base=Multiply(A=Multiply(A=@n22, B=@n21), B=@n18) {constB=2}, Exp=@n19)
  n24 = LinearInterpolate(A=@n10, B=@n15, Alpha=Saturate(None=Multiply(A=@n20, B=@n23)))
  Multiply(
   A=LinearInterpolate(A=OneMinus(None=@n6), Alpha=ComponentMask(None=@n7) {r=true})
   B=StaticSwitchParameter(True=LinearInterpolate(A=@n11, B=@n12, Alpha=@n13), False=StaticSwitchParameter(True=@n15, False=@n24) {parameterName=IsRealtimeBlur}) {parameterName=IsFog}
  )
  n25 = MaterialFunctionCall(Float2=TextureCoordinate()) {materialFunction=&BreakOutFloat2Components}
  n26 = ScalarParameter() {parameterName=FogHeight}
  n27 = ScalarParameter() {parameterName=FogRad}
  n28 = ScalarParameter() {parameterName=fogHardness}
  n29 = DepthFade(
   Opacity=Power(Base=SphereMask(A=@n25.G, B=@n26, Radius=@n27, Hardness=@n28), Exp=ScalarParameter() {defaultValue=4, parameterName=FogExp})
   FadeDistance=ScalarParameter() {defaultValue=400, parameterName=depth}
  )
  n30 = Divide(
   A=Subtract(A=WorldPosition().Z, B=ScalarParameter() {defaultValue=12, parameterName=BlendOffset})
   B=ScalarParameter() {defaultValue=8, parameterName=BlendDistance}
  )
  n31 = DepthFade(Opacity=OneMinus(None=Saturate(None=@n30)), FadeDistance=ScalarParameter() {defaultValue=28, parameterName=depthStylised})
  n32 = ComponentMask(None=CollectionParameter() {collection=&MPC_LobbyBackground, parameterName=FloorStylised}) {r=true}
  Saturate(None=StaticSwitchParameter(True=Constant() {r=1}, False=LinearInterpolate(A=@n29, B=@n31, Alpha=@n32)) {parameterName=IsRealtimeBlur})
  n33 = CollectionParameter() {collection=&MPC_LobbyBackground, parameterName=FloorStylised}
  n34 = Subtract(A=CameraPositionWS(), B=WorldPosition().XYZ)
  n35 = StaticSwitchParameter(True=Constant3Vector(), False=Multiply(A=Normalize(VectorInput=@n34), B=ScalarParameter() {defaultValue=-500, parameterName=cameraOffset})) {parameterName=IsRealtimeBlur}
  StaticSwitchParameter(
   True=LinearInterpolate(A=@n35, B=VectorParameter() {parameterName=Offset, defaultValue=<rgba 0 30 0 1>}.RGB, Alpha=ComponentMask(None=@n33) {r=true})
   False=@n35
  ) {parameterName=IsFog}
 ]
}
