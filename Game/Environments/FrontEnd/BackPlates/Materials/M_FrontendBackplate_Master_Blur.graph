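# Unlit master material for front-end backplates that shows a blurred version of a backplate texture.
# The engine's SpiralBlur-Texture function samples a texture parameter at the mesh UVs, with blur distance and distance step count exposed as scalar parameters.
# Radial steps, radial offset and kernel power stay at the function's defaults, and the multi-tap blur runs in the material itself.
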Material /Game/Environments/FrontEnd/BackPlates/Materials/M_FrontendBackplate_Master_Blur
graph {
 properties {pixelDepthOffsetMode=PDOM_Legacy, shadingModel=MSM_Unlit, usedShadingModels=Unlit}
 nodes [MaterialFunctionCall!(
   TextureObject=TextureObjectParameter() {parameterName=Param, texture=/Game/Environments/FrontEnd/BackPlates/Textures/T_protofrontend_07}
   UV=TextureCoordinate()
   Distance=ScalarParameter() {defaultValue=0.05, parameterName=Distance}
   Distance Steps=ScalarParameter() {defaultValue=32, parameterName=DistanceSteps}
   Radial Steps
   Radial Offset
   Kernel Power
  ) {materialFunction=/Engine/Functions/Engine_MaterialFunctions01/Texturing/SpiralBlur-Texture}]
}
