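# Translucent unlit material for the Battle Royale lobby backplate "hex donut", drawing a sparkling hexagonal grid with a ColorA tint parameter.
# Tiled rotated line masks, a Hexagon function and sine-based dots mask two opposite-panning screen-space noise samples.
# A screen-space radial vignette shapes the result, which MPC_LobbyBackground's Fade_Triangle_Grid fades unless a static switch bypasses it.
# A time-driven axis rotation uses PreviousFrameSwitch; a separate blink branch does not feed the faded result.
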Material /Game/Environments/FrontEnd/BackPlates/Materials/M_HexDonut_BRLobby
graph {
 properties {blendMode=BLEND_Translucent, pixelDepthOffsetMode=PDOM_Legacy, shadingModel=MSM_Unlit, usedShadingModels=Unlit}
 nodes [
  n1 = Subtract(A=ComponentMask() {g=true}) {constB=0.5}
  n2 = Frac(None=TextureCoordinate() {uTiling=4, vTiling=4})
  n3 = Subtract(A=Rotator(Coordinate=@n2, Time=Constant() {r=0.6667}) {speed=6.2832}) {constB=0.5}
  n4 = Constant() {r=0.01}
  n5 = Constant() {r=0.001}
  n6 = MaterialFunctionCall(Alpha=ComponentMask(None=Abs(None=@n3)) {r=true}, Min=@n4, Max=@n5) {materialFunction=&SmoothStep}
  Multiply(A=OneMinus(None=Ceil(None=@n1)), B=@n6.Result)
  n7 = TextureCoordinate()
  n8 = Constant() {r=4}
  n9 = Sine(None=Add(A=Multiply(A=@n7, B=@n8), B=Constant2Vector() {g=3, r=1}) {constB=0}) {period=2}
  n10 = Constant()
  n11 = MaterialFunctionCall(Alpha=DotProduct(A=ComponentMask(None=@n9) {r=true}, B=ComponentMask(None=@n9) {g=true}), Min=@n10, Max=@n10) {materialFunction=&SmoothStep}
  n12 = Multiply(A=Time(), B=ScalarParameter() {defaultValue=0.5, parameterName=BlinkSpeed, sliderMax=1}) {constB=0.125}
  n13 = Multiply(A=@n12) {constB=0.25}
  n14 = Clamp(None=Add(A=Sine(None=@n13)))
  n15 = Constant() {r=1}
  n16 = Constant()
  n17 = Constant() {r=1}
  n18 = Constant() {r=0.4325}
  n19 = Constant() {r=0.125}
  n20 = Fmod(A=@n12, B=Constant() {r=1})
  n21 = Clamp(None=@n20)
  n22 = OneMinus(None=MaterialFunctionCall!(Hex Scale=LinearInterpolate(A=@n18, B=@n19, Alpha=@n21)) {materialFunction=/Game/UI/Foundation/Materials/Hexagon}.Result)
  n23 = OneMinus(None=@n20)
  Multiply(
   A=LinearInterpolate(A=@n11.Result, B=OneMinus(None=@n11.Result), Alpha=If(A=@n14, B=@n15, "A > B"=@n16, "A == B"=@n16, "A < B"=@n17))
   B=Multiply(A=Multiply(A=@n22, B=@n23), B=ScalarParameter() {defaultValue=1, parameterName=BlinkOpacity, sliderMax=1})
  )
  n24 = Constant3Vector() {constant=<rgba 0 0 1 1>}
  n25 = Add(
   A=Multiply(A=ScalarParameter() {defaultValue=0.0025, parameterName=DonutRotationSpeed}, B=Time())
   B=Constant3Vector() {constant=<rgba 0 1 0 1>}
  )
  n26 = ObjectPositionWS()
  n27 = WorldPosition()
  PreviousFrameSwitch(
   Current Frame=RotateAboutAxis(NormalizedRotationAxis=@n24, RotationAngle=@n25, PivotPoint=@n26, Position=@n27.XYZ)
   Previous Frame=RotateAboutAxis(NormalizedRotationAxis=@n24, RotationAngle=Subtract(A=@n25) {constB=0.99}, PivotPoint=@n26, Position=@n27.XYZ)
  )
  n28 = ScreenPosition()
  n29 = Multiply(A=@n28.ViewportUV, B=ScalarParameter() {defaultValue=1, parameterName=SparkleTile, sliderMax=1})
  n30 = Add(
   A=TextureSample(UVs=Panner(Coordinate=@n29) {speedX=-0.08, speedY=0.06}) {samplerType=SAMPLERTYPE_Grayscale, texture=&LowResBlurredNoise}.RGB
   B=TextureSample(UVs=Panner(Coordinate=@n29) {speedX=0.06, speedY=-0.08}) {samplerType=SAMPLERTYPE_Grayscale, texture=&LowResBlurredNoise}.RGB
  )
  n31 = ScalarParameter() {defaultValue=8, parameterName=SparklePower}
  n32 = Multiply(
   A=Clamp(None=Power(Base=@n30, Exp=@n31) {constExponent=8}, Min=ScalarParameter() {defaultValue=0.1, parameterName=SparkleBaseOpacity}) {minDefault=0.1}
   B=ScalarParameter() {defaultValue=1, parameterName=SparkleOpacity, sliderMax=1}
  )
  n33 = Rotator(Coordinate=@n2, Time=Constant() {r=0.3333}) {speed=6.2832}
  n34 = ComponentMask(None=Abs(None=Subtract(A=@n33) {constB=0.5})) {r=true}
  n35 = Abs(None=Subtract(A=@n2) {constB=0.5})
  n36 = Add(
   A=Add(
    A=@n6.Result
    B=MaterialFunctionCall(Alpha=@n34, Min=@n4, Max=@n5) {materialFunction=&SmoothStep}.Result
   )
   B=MaterialFunctionCall(Alpha=ComponentMask(None=@n35) {r=true}, Min=@n4, Max=@n5) {materialFunction=&SmoothStep}.Result
  )
  n37 = Constant() {r=0.425}
  n38 = Multiply(
   A=Add(A=Clamp(None=@n36), B=MaterialFunctionCall(Hex Scale=@n37) {materialFunction=&Hexagon}.Result)
   B=ScalarParameter() {defaultValue=0.4, parameterName=PrimaryEdgeOpacity, sliderMax=1}
  )
  n39 = Sine(None=@n2) {period=2}
  n40 = ScalarParameter() {defaultValue=0.99, parameterName=DotSizeSSMin}
  n41 = ScalarParameter() {defaultValue=0.995, parameterName=DotSizeSSMax}
  n42 = MaterialFunctionCall(Alpha=DotProduct(A=ComponentMask(None=@n39) {r=true}, B=ComponentMask(None=@n39) {g=true}), Min=@n40, Max=@n41) {materialFunction=&SmoothStep}
  n43 = Sine(None=Add(A=@n2, B=Constant2Vector() {g=0.5})) {period=2}
  n44 = DotProduct(A=ComponentMask(None=@n43) {r=true}, B=ComponentMask(None=@n43) {g=true})
  n45 = Sine(None=Add(A=@n2, B=Constant2Vector() {g=-0.5})) {period=2}
  n46 = DotProduct(A=ComponentMask(None=@n45) {r=true}, B=ComponentMask(None=@n45) {g=true})
  n47 = MaterialFunctionCall(Alpha=Add(A=Clamp(None=@n44), B=Clamp(None=@n46)), Min=@n40, Max=@n41) {materialFunction=&SmoothStep}
  n48 = Sine(None=Add(A=@n2, B=Constant2Vector() {g=-0.25, r=0.4336})) {period=2}
  n49 = ComponentMask(None=@n48) {r=true}
  n50 = ComponentMask(None=@n48) {g=true}
  n51 = Sine(None=Add(A=@n2, B=Constant2Vector() {g=-0.25, r=-0.4336})) {period=2}
  n52 = ComponentMask(None=@n51) {r=true}
  n53 = ComponentMask(None=@n51) {g=true}
  n54 = Add(A=Clamp(None=DotProduct(A=@n49, B=@n50)), B=Clamp(None=DotProduct(A=@n52, B=@n53)))
  n55 = Sine(None=Add(A=@n2, B=Constant2Vector() {g=0.25, r=0.4336})) {period=2}
  n56 = Clamp(None=DotProduct(A=ComponentMask(None=@n55) {r=true}, B=ComponentMask(None=@n55) {g=true}))
  n57 = Sine(None=Add(A=@n2, B=Constant2Vector() {g=0.25, r=-0.4336})) {period=2}
  n58 = Clamp(None=DotProduct(A=ComponentMask(None=@n57) {r=true}, B=ComponentMask(None=@n57) {g=true}))
  n59 = Add(
   A=Add(
    A=Add(A=@n42.Result, B=@n47.Result) {constB=0}
    B=MaterialFunctionCall(Alpha=@n54, Min=@n40, Max=@n41) {materialFunction=&SmoothStep}.Result
   ) {constB=0}
   B=MaterialFunctionCall(Alpha=Add(A=@n56, B=@n58), Min=@n40, Max=@n41) {materialFunction=&SmoothStep}.Result
  ) {constB=0}
  n60 = OneMinus(None=Add(A=Clamp(None=@n38), B=Clamp(None=@n59)))
  n61 = Constant2Vector() {g=0.5, r=0.5}
  n62 = ScalarParameter() {defaultValue=0.5, parameterName=VignetteRadius}
  n63 = ScalarParameter() {defaultValue=8, parameterName=VignetteDensity}
  n64 = Multiply(
   A=Multiply(A=LinearInterpolate(A=@n32, Alpha=@n60) {constB=0}, B=ScalarParameter() {defaultValue=1, parameterName=OverallOpacity, sliderMax=1})
   B=Multiply(
    A=MaterialFunctionCall(UVs=@n28.ViewportUV, CenterPosition=@n61, Radius=@n62, Density=@n63) {materialFunction=&RadialGradientExponential}.RadialGradientExponential
    B=ScalarParameter() {defaultValue=1, parameterName=VignetteOpacity, sliderMax=1}
   )
  )
  StaticSwitchParameter(
   True=@n64
   False=LinearInterpolate(
    A=Constant()
    B=@n64
    Alpha=CollectionParameter() {collection=&/MPC_LobbyBackground, parameterName=Fade_Triangle_Grid}
   )
  ) {parameterName=Always Enabled no MPCControl}
  VectorParameter() {parameterName=ColorA, defaultValue=<rgba 0.115 0.6283 1 1>}
 ]
}
