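# Translucent unlit material for the animated triangles in the Fortnite frontend lobby backdrop.
# Time, randomized by vertex color and object position, drives vertex motion: a static switch picks rotate-plus-sine-shift or offset-pivot spin, plus a collection-driven Y displacement.
# Color projects a blurred backplate render target and default background texture by lobby-centered world position, blended by a collection Fade.
# An edge-mask texture scales opacity; a distance-based two-color gradient is also computed.
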Material /Game/Environments/FrontEnd/BackPlates/Materials/M_Lobby_Triangle
graph {
 properties {blendMode=BLEND_Translucent, pixelDepthOffsetMode=PDOM_Legacy, shadingModel=MSM_Unlit, usedShadingModels=Unlit}
 nodes [
  n1 = VertexColor()
  n2 = Multiply(A=ScalarParameter() {defaultValue=14.6149, parameterName=Random2}, B=@n1.R)
  n3 = Multiply(A=ScalarParameter() {defaultValue=144.5565, parameterName=Random3}, B=@n1.G)
  n4 = MaterialFunctionCall(Float3=ObjectPositionWS()) {materialFunction=&BreakOutFloat3Components}
  n5 = Add(A=Add(A=@n4.R, B=@n4.R), B=@n4.B)
  n6 = Frac(None=Add(A=Add(A=@n2, B=@n3), B=Divide(A=@n5) {constB=100}))
  n7 = Add(A=Time(), B=Multiply(A=@n6, B=ScalarParameter() {defaultValue=63.5312, parameterName=Random}))
  n8 = Sine(None=Multiply(A=@n7, B=ScalarParameter() {defaultValue=-0.05, parameterName=Square_Speed3}))
  Abs(None=@n8)
  n9 = Divide(
   A=TransformPosition(None=WorldPosition() {worldPositionShaderOffset=WPT_ExcludeAllShaderOffsets}.XYZ) {transformSourceType=TRANSFORMPOSSOURCE_World}
   B=ObjectBounds()
  )
  n10 = ComponentMask(None=@n9) {g=true}
  n11 = Normalize(VectorInput=VertexNormalWS())
  n12 = MaterialFunctionCall() {materialFunction=/Engine/Functions/Engine_MaterialFunctions02/WorldPositionOffset/ObjectScale}
  n13 = ScalarParameter() {defaultValue=0.08, parameterName=Square_Speed}
  n14 = LinearInterpolate(B=Multiply(A=ScalarParameter() {defaultValue=0.005, parameterName=RotateMax}, B=@n12.Scale Z), Alpha=Sine(None=Multiply(A=@n7, B=@n13)))
  n15 = ObjectPositionWS()
  n16 = WorldPosition()
  n17 = Normalize(VectorInput=VertexNormalWS())
  n18 = ScalarParameter() {defaultValue=0.07, parameterName=Square_Speed2}
  n19 = LinearInterpolate(B=Multiply(A=ScalarParameter() {defaultValue=0.003, parameterName=RotateMax2}, B=@n12.Scale Z), Alpha=Sine(None=Multiply(A=@n7, B=@n18)))
  n20 = ObjectPositionWS()
  n21 = WorldPosition()
  n22 = Add(
   A=Multiply(A=@n10, B=RotateAboutAxis(NormalizedRotationAxis=@n11, RotationAngle=@n14, PivotPoint=@n15, Position=@n16.XYZ))
   B=Multiply(A=OneMinus(None=@n10), B=RotateAboutAxis(NormalizedRotationAxis=@n17, RotationAngle=@n19, PivotPoint=@n20, Position=@n21.XYZ))
  )
  n23 = Constant3Vector() {constant=<rgba 0 1 0 0>}
  n24 = ScalarParameter() {defaultValue=3.4029, parameterName=SineMovement}
  n25 = Divide(A=Add(A=@n8))
  n26 = Multiply(A=VertexColor().R, B=Multiply(A=Transform(None=@n23) {transformSourceType=TRANSFORMSOURCE_Local}, B=LinearInterpolate(B=@n24, Alpha=@n25)))
  n27 = VertexNormalWS()
  n28 = Normalize(VectorInput=@n27)
  n29 = ScalarParameter() {defaultValue=0.1184, parameterName=Speed}
  n30 = ScalarParameter() {defaultValue=-0.1143, parameterName=Speed2}
  n31 = Frac(None=Multiply(A=Add(B=@n7), B=LinearInterpolate(A=@n29, B=@n30, Alpha=@n6)))
  n32 = WorldPosition()
  n33 = Subtract(A=ObjectPositionWS(), B=WorldPosition().XYZ)
  n34 = ScalarParameter() {defaultValue=1.7165, parameterName=RotationCenteroffset}
  n35 = MaterialFunctionCall() {materialFunction=&ObjectScale}
  n36 = Add(A=@n32.XYZ, B=Multiply(A=Normalize(VectorInput=@n33), B=Multiply(A=@n34, B=@n35.Scale Z)))
  Add(
   A=StaticSwitchParameter(True=Add(A=@n22, B=@n26), False=RotateAboutAxis(NormalizedRotationAxis=@n28, RotationAngle=@n31, PivotPoint=@n36, Position=@n32.XYZ)) {parameterName=UseShiftorRotate}
   B=Multiply(
    A=CollectionParameter() {collection=/Game/Packages/Fortnite_UniqueMaterials_1/PostProcess/MPC_PostProcessControl, parameterName=LobbyTriangles}
    B=Constant3Vector() {constant=<rgba 0 2000 0 0>}
   )
  )
  ComponentMask(None=Multiply(A=@n9) {constB=-1}) {b=true}
  CrossProduct(A=@n27, B=Constant3Vector() {constant=<rgba 0.1 -0.1 -1 0>})
  n37 = WorldPosition()
  n38 = VectorParameter() {parameterName=LobbyCenterPostion, defaultValue=<rgba 106.6131 144647.25 -646.5316 1>}
  n39 = Divide(
   A=Distance(A=ComponentMask(None=@n37.XYZ) {b=true, r=true}, B=ComponentMask(None=@n38.RGB) {b=true, r=true})
   B=ScalarParameter() {defaultValue=552.914, parameterName=FadeDistance}
  )
  n40 = ScalarParameter() {defaultValue=0.6073, parameterName=FadeDistancePower}
  LinearInterpolate(
   A=VectorParameter() {parameterName=Color, defaultValue=<rgba 0.4029 1.186 2 1>}.RGB
   B=VectorParameter() {parameterName=Color2, defaultValue=<rgba 0 0.1195 0.2747 1>}.RGB
   Alpha=Saturate(None=Power(Base=@n39, Exp=@n40))
  )
  MaterialFunctionCall() {materialFunction=&BoundingBoxBased_0-1_UVW}
  n41 = ComponentMask(None=WorldPosition().XYZ) {b=true, r=true}
  n42 = ComponentMask(None=VectorParameter() {parameterName=LobbyCenterPostion, defaultValue=<rgba 106.6131 144647.25 -646.5316 1>}.RGB) {b=true, r=true}
  n43 = ScalarParameter() {defaultValue=958.3593, parameterName=BaseUVMatch_X}
  n44 = ScalarParameter() {defaultValue=1353.2234, parameterName=BaseUVMatch_Y}
  n45 = MaterialFunctionCall(
   UVs=MaterialFunctionCall(UVs=Subtract(A=@n41, B=@n42), Texture Scale=AppendVector(A=@n43, B=@n44)) {materialFunction=&ScaleUVsByCenter}.UVs
   Texture Scale=AppendVector(A=ScalarParameter() {defaultValue=1.1631, parameterName=ScaleX}, B=ScalarParameter() {defaultValue=0.6714, parameterName=ScaleY})
  ) {materialFunction=&ScaleUVsByCenter}
  n46 = TextureSample(UVs=@n45.UVs) {texture=&/RT_Frontend_Backplate_Blur}
  n47 = TextureSample(UVs=@n45.UVs) {texture=/Game/Environments/FrontEnd/BackPlates/Textures/T_Background_LobbyDefault_SF}
  n48 = CollectionParameter() {collection=&/MPC_LobbyBackground, parameterName=Fade}
  Multiply(
   A=Power(Base=LinearInterpolate(A=@n46.RGB, B=@n47.RGB, Alpha=@n48), Exp=ScalarParameter() {defaultValue=1.05, parameterName=ColorPower})
   B=ScalarParameter() {defaultValue=1.1, parameterName=ColorBrightness}
  )
  n49 = TextureSample() {texture=/Game/Environments/FrontEnd/BackPlates/Textures/T_Edgemask_M}
  Multiply(
   A=StaticSwitchParameter(True=@n49.R, False=@n49.R) {parameterName=UseShiftorRotate}
   B=ScalarParameter() {defaultValue=0.9421, parameterName=Opacity}
  )
  Saturate(None=@n8)
  Time()
  Time()
  Time()
  WorldPosition()
 ]
}
