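# Unlit material for a front-end backplate, named as a simple store background, that references the default lobby background texture.
# Its single texture sample takes viewport UVs from screen position, so the image maps to the screen rather than to mesh UVs.
# The graph shows no other processing, parameters or animation.
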Material /Game/Environments/FrontEnd/BackPlates/Materials/M_SimpleStoreBackgorund
graph {
 properties {pixelDepthOffsetMode=PDOM_Legacy, shadingModel=MSM_Unlit, usedShadingModels=Unlit}
 nodes [TextureSample(UVs=ScreenPosition().ViewportUV) {texture=/Game/Environments/FrontEnd/BackPlates/Textures/T_Background_LobbyDefault}]
}
